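# Sound cue for the PoppySoap announcer's "Flawless" voice line, playing a single wave asset.
# It routes to the Fort_UI_InGame_Critical sound class with a boosted volume, virtualization disabled and a very large max distance.
# An announcer VO concurrency set is assigned, alongside concurrency override values of one instance that prevents new plays.
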SoundCue /PoppySoap/Assets/Sound/IP_SC_PoppySoap_Announcer_VO_Flawless
root=WavePlayer_4
nodes [{id=WavePlayer_4, properties {soundWaveAssetPtr=&/IP_PoppySoap_VO_Announcer_Flawless_0101}}]
properties {
 maxDistance=2097152
 soundClassObject=&Fort_UI_InGame_Critical
 virtualizationMode=Disabled
 volumeMultiplier=1.6
 concurrencyOverrides {maxCount=1, resolutionRule=PreventNew}
 concurrencySet [&/CONC_IP_PoppySoap_VO_Announcer]
}
