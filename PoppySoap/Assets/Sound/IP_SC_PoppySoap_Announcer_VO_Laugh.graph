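# Sound cue for the PoppySoap announcer's laugh voice-over.
# A random node picks one of three equally weighted laugh recordings, with branch culling excluded on that node.
# It plays through the critical in-game UI sound class with boosted volume, a very large max distance and virtualization disabled.
# A dedicated PoppySoap announcer concurrency set governs overlap.
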SoundCue /PoppySoap/Assets/Sound/IP_SC_PoppySoap_Announcer_VO_Laugh
root=Random_0
nodes [
 {id=Random_0, children [WavePlayer_7,WavePlayer_8,WavePlayer_9], properties {bShouldExcludeFromBranchCulling=true, weights [1,1,1]}}
 {id=WavePlayer_7, properties {soundWaveAssetPtr=&/IP_PoppySoap_VO_Announcer_Laugh_01}}
 {id=WavePlayer_8, properties {soundWaveAssetPtr=&/IP_PoppySoap_VO_Announcer_Laugh_02}}
 {id=WavePlayer_9, properties {soundWaveAssetPtr=&/IP_PoppySoap_VO_Announcer_Laugh_03}}
]
properties {
 maxDistance=2097152
 soundClassObject=&Fort_UI_InGame_Critical
 virtualizationMode=Disabled
 volumeMultiplier=1.6
 concurrencySet [&/CONC_IP_PoppySoap_VO_Announcer]
}
