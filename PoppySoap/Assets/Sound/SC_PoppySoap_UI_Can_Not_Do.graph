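# UI sound cue for a "can't do" error in PoppySoap, in the critical in-game UI sound class.
# A branch on IsLocal plays the sound when true or unset and stays silent when false.
# After a short delay, it layers a pitched-up pistol foley click with a quieter error ping.
# Overridden concurrency caps overlapping instances and stops the oldest.
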SoundCue /PoppySoap/Assets/Sound/SC_PoppySoap_UI_Can_Not_Do
root=Branch_0
nodes [
 {id=Branch_0, children [Mixer_1,null,Mixer_1], properties {boolParameterName=IsLocal}}
 {id=Delay_0, children [Mixer_0], properties {delayMax=0.2, delayMin=0.2}}
 {id=Mixer_0, children [Modulator_0,WavePlayer_0], properties {inputVolume [1.25,0.5]}}
 {id=Mixer_1, children [Delay_0,null], properties {inputVolume [1,1]}}
 {id=Modulator_0, children [WavePlayer_2], properties {pitchMax=1.3, pitchMin=1.3}}
 {id=WavePlayer_0, properties {soundWaveAssetPtr=/Game/Sounds/Fort_UI_Sounds/Gameplay/UI_Error_Ping_Cant_Do_Default_01}}
 {id=WavePlayer_2, properties {soundWaveAssetPtr=/Game/Sounds/Fort_Weapons/pistol/pistol_foley/auto_sdk_end_01}}
]
properties {
 bOverrideConcurrency=true
 maxDistance=2097152
 soundClassObject=&Fort_UI_InGame_Critical
 volumeMultiplier=1
 concurrencyOverrides {maxCount=2, resolutionRule=StopOldest, retriggerTime=0.6}
}
